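# Looping exterior distance cue for a spot blimp POI, judging by its name the front, focused variant, limited to two concurrent voices.
# Distant, medium-distant and very-distant loops each get their own attenuation and are mixed at different levels through a modulator.
# Nested looping Envelopers slowly drift each layer's pitch and volume, with a further pitch drift on the mix and two-second fade-ins at both stages.
# The first two quality slots share this chain; the other two are empty.
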SoundCue /Game/Sounds/Artemis/POI/Blimps/Spot_Blimp_Exterior_Distance_Focus_Front_Cue
root=QualityLevel_0
nodes [
 {
  id=Attenuation_0
  children [Enveloper_7]
  properties {attenuationSettings=&/Spot_Blimp_Exterior_Distant_Custom_Att}
 }
 {
  id=Attenuation_1
  children [Enveloper_8]
  properties {attenuationSettings=&/Spot_Blimp_Exterior_Medium_Distant_Custom_Att}
 }
 {
  id=Attenuation_2
  children [Enveloper_9]
  properties {attenuationSettings=&/Spot_Blimp_Exterior_Very_Distant_Custom_Att}
 }
 {
  id=Enveloper_0
  children [WavePlayer_2]
  properties {
   bLoop=true
   bLoopIndefinitely=true
   loopEnd=30
   pitchCurve {editorCurveData {keys [{interpMode=RCIM_Cubic, time=0}, <at 15.9468 0.9181 RCIM_Cubic>, {interpMode=RCIM_Cubic, time=30}]}}
   volumeCurve {editorCurveData {keys [<at 0 1>]}}
  }
 }
 {
  id=Enveloper_1
  children [WavePlayer_3]
  properties {
   bLoop=true
   bLoopIndefinitely=true
   loopEnd=30
   pitchCurve {editorCurveData {keys [{interpMode=RCIM_Cubic, time=0}, <at 15.9468 0.9181 RCIM_Cubic>, {interpMode=RCIM_Cubic, time=30}]}}
   volumeCurve {editorCurveData {keys [<at 0 1>]}}
  }
 }
 {
  id=Enveloper_2
  children [Enveloper_1]
  properties {
   bLoop=true
   bLoopIndefinitely=true
   loopEnd=20
   pitchCurve {editorCurveData {keys [{time=0}]}}
   volumeCurve {editorCurveData {keys [
      {arriveTangent=-0.01, interpMode=RCIM_Cubic, leaveTangent=-0.01, time=0, value=1}
      <at 10 0.9 RCIM_Cubic>
      {arriveTangent=0.01, interpMode=RCIM_Cubic, leaveTangent=0.01, time=20, value=1}
     ]}}
  }
 }
 {
  id=Enveloper_27
  children [Modulator_1]
  properties {
   bLoop=true
   bLoopIndefinitely=true
   loopEnd=60
   pitchCurve {editorCurveData {keys [{interpMode=RCIM_Cubic, time=0}, <at 30 0.95 RCIM_Cubic>, {interpMode=RCIM_Cubic, time=60}]}}
   volumeCurve {editorCurveData {keys [<at 0 1>]}}
  }
 }
 {
  id=Enveloper_29
  children [Enveloper_27]
  properties {
   pitchCurve {editorCurveData {keys [{time=0}]}}
   volumeCurve {editorCurveData {keys [{arriveTangent=1.0676, interpMode=RCIM_Cubic, leaveTangent=1.0676, tangentMode=RCTM_User, time=0, value=0}, <at 2 1 RCIM_Cubic>]}}
  }
 }
 {
  id=Enveloper_3
  children [Enveloper_0]
  properties {
   bLoop=true
   bLoopIndefinitely=true
   loopEnd=20
   pitchCurve {editorCurveData {keys [{time=0}]}}
   volumeCurve {editorCurveData {keys [
      {arriveTangent=-0.01, interpMode=RCIM_Cubic, leaveTangent=-0.01, time=0, value=1}
      <at 10 0.9 RCIM_Cubic>
      {arriveTangent=0.01, interpMode=RCIM_Cubic, leaveTangent=0.01, time=20, value=1}
     ]}}
  }
 }
 {
  id=Enveloper_4
  children [WavePlayer_4]
  properties {
   bLoop=true
   bLoopIndefinitely=true
   loopEnd=30
   pitchCurve {editorCurveData {keys [{interpMode=RCIM_Cubic, time=0}, <at 15.9468 0.9181 RCIM_Cubic>, {interpMode=RCIM_Cubic, time=30}]}}
   volumeCurve {editorCurveData {keys [<at 0 1>]}}
  }
 }
 {
  id=Enveloper_5
  children [Enveloper_4]
  properties {
   bLoop=true
   bLoopIndefinitely=true
   loopEnd=20
   pitchCurve {editorCurveData {keys [{time=0}]}}
   volumeCurve {editorCurveData {keys [
      {arriveTangent=-0.01, interpMode=RCIM_Cubic, leaveTangent=-0.01, time=0, value=1}
      <at 10 0.9 RCIM_Cubic>
      {arriveTangent=0.01, interpMode=RCIM_Cubic, leaveTangent=0.01, time=20, value=1}
     ]}}
  }
 }
 {
  id=Enveloper_7
  children [Enveloper_3]
  properties {
   pitchCurve {editorCurveData {keys [{time=0}]}}
   volumeCurve {editorCurveData {keys [{arriveTangent=1.0676, interpMode=RCIM_Cubic, leaveTangent=1.0676, tangentMode=RCTM_User, time=0, value=0}, <at 2 1 RCIM_Cubic>]}}
  }
 }
 {
  id=Enveloper_8
  children [Enveloper_2]
  properties {
   pitchCurve {editorCurveData {keys [{time=0}]}}
   volumeCurve {editorCurveData {keys [{arriveTangent=1.0676, interpMode=RCIM_Cubic, leaveTangent=1.0676, tangentMode=RCTM_User, time=0, value=0}, <at 2 1 RCIM_Cubic>]}}
  }
 }
 {
  id=Enveloper_9
  children [Enveloper_5]
  properties {
   pitchCurve {editorCurveData {keys [{time=0}]}}
   volumeCurve {editorCurveData {keys [{arriveTangent=1.0676, interpMode=RCIM_Cubic, leaveTangent=1.0676, tangentMode=RCTM_User, time=0, value=0}, <at 2 1 RCIM_Cubic>]}}
  }
 }
 {id=Mixer_1, children [Attenuation_0, Attenuation_1, Attenuation_2], properties {inputVolume [0.8,0.5,0.15]}}
 {id=Modulator_1, children [Mixer_1], properties {pitchMax=1, pitchMin=1}}
 {id=QualityLevel_0, children [Enveloper_29,Enveloper_29,null,null]}
 {id=WavePlayer_2, properties {bLooping=true, soundWaveAssetPtr=&/Spot_Blimp_Exterior_Distant_Loop}}
 {id=WavePlayer_3, properties {bLooping=true, soundWaveAssetPtr=&/Spot_Blimp_Exterior_Medium_Distant_Loop}}
 {id=WavePlayer_4, properties {bLooping=true, soundWaveAssetPtr=&/Spot_Blimp_Exterior_Very_Distant_Loop}}
]
properties {
 bEnableBusSends=false
 bEnableSubmixSends=false
 bOverrideConcurrency=true
 maxDistance=28000
 soundClassObject=&Fort_Gadgets
 volumeMultiplier=0.22
 concurrencyOverrides {maxCount=2, voiceStealReleaseTime=0.2}
}
